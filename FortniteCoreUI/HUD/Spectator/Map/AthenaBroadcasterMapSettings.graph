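# Map settings panel for Fortnite's broadcaster/spectator HUD, controlling minimap players, weapon tracers, vehicles, bus paths and name zoom level.
# Rotators and slider write to the minimap spectating component; Refresh Options reads it back on activation.
# Turning players off collapses tracer, vehicle and name-zoom rows; slider inversely maps to zoom threshold, with zero setting 9999.
# Added map icons create buttons mapped to actor classes, toggling visibility through the replay context and dimming when off.
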WidgetBlueprint /FortniteCoreUI/HUD/Spectator/Map/AthenaBroadcasterMapSettings
graphs [
 EventGraph {nodes [
   n1 = "Selection Changed (WeaponTracersRotator)"!() -> (OutputDelegate, then, Selected Index:int=0) {is=title K2Node_ComponentBoundEvent}
   > n2 = Cast To FortPlayerController(Object=GetOwningPlayer())
   > n3 = Cast To FortControllerComponent_MinimapSpectating!(
    execute
    Object=GetMiniMapComponent!(self:Fort Player Controller Object Reference=@n2) -> (ReturnValue:Fort Controller Component Minimap Object Reference)
   ) -> (then, CastFailed, AsFort Controller Component Minimap Spectating:Fort Controller Component Minimap Spectating Object Reference) {is=title K2Node_DynamicCast}
   > SetMapWeaponTracersEnabled!(
    execute
    self:Fort Controller Component Minimap Spectating Object Reference=@n3
    bEnable:bool="integer > integer"(A=@n1.Selected Index, B=0) {is=title K2Node_CallFunction}
   ) -> (then)
   n4 = "Slider Changed (NameZoomVisibility)"!() -> (OutputDelegate, then, Slider Value:double) {is=title K2Node_ComponentBoundEvent}
   > n5 = Cast To FortPlayerController(Object=GetOwningPlayer())
   > n6 = Cast To FortControllerComponent_MinimapSpectating(Object=GetMiniMapComponent(self=@n5))
   n7 = float + float(
    A=GetMaxMapZoomLevel!(self:Fort Replay Context Object Reference=Get FortReplayContext!() -> (ReturnValue:Fort Replay Context Object Reference) {is=title K2Node_GetBlueprintContextEx}) -> (ReturnValue:float)
    B=0.01
   ) {is=title K2Node_CommutativeAssociativeBinaryOperator}
   > Set ZoomThresholdForPlayerNames!(
    execute
    ZoomThresholdForPlayerNames:float=Select(
     Option 0:double=MapRangeClamped(Value=@n4.Slider Value, InRangeA=1, InRangeB=0, OutRangeA=1, OutRangeB=@n7)
     Option 1:double=9999
     Index="Equal (Float)"!(A:double=@n4.Slider Value, B:double=0) -> (ReturnValue:bool) {is=title K2Node_CallFunction}
    ) -> (ReturnValue:double)
    self:Fort Controller Component Minimap Spectating Object Reference=@n6
   ) -> (then, Output_Get:float)
   n8 = "Selection Changed (VehicleIconsRotator)"!() -> (OutputDelegate, then, Selected Index:int=0) {is=title K2Node_ComponentBoundEvent}
   > n9 = Cast To FortPlayerController(Object=GetOwningPlayer())
   > n10 = Cast To FortControllerComponent_MinimapSpectating(Object=GetMiniMapComponent(self=@n9))
   > SetShowVehiclesOnMap!(
    execute
    self:Fort Controller Component Minimap Spectating Object Reference=@n10
    bInShowVehiclesOnMap:bool="integer > integer"(A=@n8.Selected Index, B=0)
   ) -> (then)
   n11 = "Selection Changed (PlayerIndicatorsRotator)"!() -> (OutputDelegate, then, Selected Index:int=0) {is=title K2Node_ComponentBoundEvent}
   > n12 = Cast To FortPlayerController(Object=GetOwningPlayer())
   > n13 = Cast To FortControllerComponent_MinimapSpectating(Object=GetMiniMapComponent(self=@n12))
   n14 = "integer > integer"(A=@n11.Selected Index, B=0)
   > SetShowPlayersOnMap!(execute, self:Fort Controller Component Minimap Spectating Object Reference=@n13, bInShowPlayersOnMap:bool=@n14) -> (then)
   > SetVisibility(
    self=[
     Get WeaponTracersRotator() -> (WeaponTracersRotator:Rotator Selector Replay Settings Object Reference)
     Get NameZoomVisibility() -> (NameZoomVisibility:Options Menu Slider Replay Object Reference)
     Get VehicleIconsRotator() -> (VehicleIconsRotator:Rotator Selector Replay Settings Object Reference)
    ]
    InVisibility=Select(Option 0:ESlateVisibility Enum=Collapsed, Option 1:ESlateVisibility Enum=Visible, Index=@n14) -> (ReturnValue:ESlateVisibility Enum)
   )
   n15 = "Selection Changed (BusPathsRotator)"!() -> (OutputDelegate, then, Selected Index:int=0) {is=title K2Node_ComponentBoundEvent}
   > n16 = Cast To FortPlayerController(Object=GetOwningPlayer())
   > n17 = Cast To FortControllerComponent_MinimapSpectating(Object=GetMiniMapComponent(self=@n16))
   > SetShowBusPathsOnMap!(
    execute
    self:Fort Controller Component Minimap Spectating Object Reference=@n17
    bInShowBusPathsOnMap:bool="integer > integer"(A=@n15.Selected Index, B=0)
   ) -> (then)
   Event On Activated()
   > Refresh Options()
   n18 = Event Handle Map Icon Added!() -> (OutputDelegate, then, MapIcon:Fort Spectate Clickable Map Icon Object Reference) {is=title K2Node_Event}
   > n19 = Create Widget!(
    execute:Exec
    Class:User Widget Class Reference=/FortUILibrary/Foundation/Buttons/IconTextButtonReplay.IconTextButtonReplay_C
    OwningPlayer:Player Controller Object Reference=GetOwningPlayer()
    ButtonText:Text
    IconBrush:Slate Brush Structure=Get SpectatorMapIndicatorBrush!(self:Fort Spectate Clickable Map Icon Object Reference=@n18.MapIcon) -> (SpectatorMapIndicatorBrush:Slate Brush Structure)
    ControllerInputStyle:Common Button Style Class Reference
    Style:Common Button Style Class Reference=/FortUILibrary/Foundation/Buttons/ButtonStyles/ButtonStyle-TransparentWithCues-XS-B_OverlayColor.ButtonStyle-TransparentWithCues-XS-B_OverlayColor_C
    bApplyAlphaOnDisable:bool=False
    bLocked:bool=False
    bSelectable:bool=False
    bShouldSelectUponReceivingFocus:bool=False
    bInteractableWhenSelected:bool=False
    bToggleable:bool=False
    bTriggerClickedAfterSelection:bool=False
    bDisplayInputActionWhenNotInteractable:bool=True
    bHideInputActionWithKeyboard:bool=False
    bShouldUseFallbackDefaultInputAction:bool=True
    bRequiresHold:bool=False
    HoldData:Common UIHold Data Class Reference
    ClickMethod:EButtonClickMethod Enum=DownAndUp
    InputPriority:int=0
    TriggeringInputAction:Data Table Row Handle Structure
   ) -> (then:Exec, ReturnValue:Icon Text Button Replay Object Reference)
   > Add!(
    execute
    TargetMap:Map of Common Button Legacy Object References to Actor Class References=Get MapIconVisibility() -> (MapIconVisibility:Map of Common Button Legacy Object References to Actor Class References)
    Key:"Common Button Legacy Object Reference (by ref)"=@n19
    Value:"Actor Class Reference (by ref)"=Get Class(Object=GetOwner(self=@n18.MapIcon))
   ) -> (then) {is=title K2Node_CallFunction}
   > Bind Event to On Button Clicked(self:Common Button Legacy Object Reference=@n19, Delegate=Create Event()) {is=title K2Node_AddDelegate}
   > AddChildToHorizontalBox!(
    execute
    self:Horizontal Box Object Reference=Get ButtonBox() -> (ButtonBox:Horizontal Box Object Reference)
    Content:Widget Object Reference=@n19
   ) -> (then, ReturnValue:Horizontal Box Slot Object Reference)
   Event PreConstruct()
   > Initialize!(
    execute
    self:Rotator Selector Replay Settings Object Reference=Get WeaponTracersRotator()
    Display Text:Text="LOCTEXT(\"Map Weapon Tracers\")"
    Options:Array of Texts=Make Array!("[0]":Text="LOCTEXT(\"Off\")", "[1]":Text="LOCTEXT(\"On\")") -> (Array:Array of Texts)
    Hover Text:Text="LOCTEXT(\"Enable/Disable map weapon tracers\")"
    Tab Tooltip:Common Text Object Reference
   ) -> (then)
   > Update Slider!(
    execute
    self:Options Menu Slider Replay Object Reference=Get NameZoomVisibility()
    Slider Text:Text="LOCTEXT(\"Name Zoom Level\")"
    Slider Value:double=0.5
    Hover Text:Text
    Tooltip Text Block:Common Text Object Reference
    Min Max Value:Vector 2D Structure=(X=0,Y=1)
   ) -> (then)
   > Initialize(
    self=Get VehicleIconsRotator()
    Display Text="LOCTEXT(\"Show Vehicles\")"
    Options=Make Array("[0]"="LOCTEXT(\"Off\")", "[1]"="LOCTEXT(\"On\")")
   )
   > Initialize(
    self=Cast To RotatorSelectorReplaySettings!(Object=Get PlayerIndicatorsRotator() -> (PlayerIndicatorsRotator:Common User Widget Object Reference)) -> (AsRotator Selector Replay Settings:Rotator Selector Replay Settings Object Reference, bSuccess:bool=~) {is=title K2Node_DynamicCast}
    Display Text="LOCTEXT(\"Show Players\")"
    Options=Make Array("[0]"="LOCTEXT(\"Off\")", "[1]"="LOCTEXT(\"On\")")
   )
   > Initialize(
    self=Get BusPathsRotator() -> (BusPathsRotator:Rotator Selector Replay Settings Object Reference)
    Display Text="LOCTEXT(\"Show Bus Paths\")"
    Options=Make Array("[0]"="LOCTEXT(\"Off\")", "[1]"="LOCTEXT(\"On\")")
   )
  ]}
 Map Icon Added {nodes [
   Map Icon Added!() -> (then, Icon:Fort Spectate Clickable Map Icon Object Reference) {is=title K2Node_FunctionEntry}
   > n1 = Create Widget(
    Class=&IconTextButtonReplay_C
    OwningPlayer=GetOwningPlayer()
    IconBrush=Get SpectatorMapIndicatorBrush(self=Get Icon() -> (Icon:Fort Spectate Clickable Map Icon Object Reference))
    Style=&ButtonStyle-TransparentWithCues-XS-B_OverlayColor_C
    bApplyAlphaOnDisable=False
    bLocked=False
    bSelectable=False
    bShouldSelectUponReceivingFocus=False
    bInteractableWhenSelected=False
    bToggleable=False
    bTriggerClickedAfterSelection=False
    bDisplayInputActionWhenNotInteractable=True
    bHideInputActionWithKeyboard=False
    bShouldUseFallbackDefaultInputAction=True
    bRequiresHold=False
    ClickMethod=DownAndUp
    InputPriority=0
   )
   n2 = Get Icon()
   > Add(TargetMap=Get MapIconVisibility(), Key=@n1, Value=Get Class(Object=GetOwner(self=@n2)))
   > Bind Event to On Button Clicked(self=@n1, Delegate=Create Event())
   > AddChildToHorizontalBox(self=Get ButtonBox(), Content=@n1)
  ]}
 OnMapIconClicked {nodes [
   n1 = OnMapIconClicked!() -> (then, ButtonClicked:Common Button Legacy Object Reference) {is=title K2Node_FunctionEntry}
   n2 = Find(
    TargetMap:Map of Common Button Legacy Object References to Actor Class References=Get MapIconVisibility()
    Key:"Common Button Legacy Object Reference (by ref)"=@n1
   ) -> (Value:Actor Class Reference)
   > Branch(Condition=@n2.ReturnValue)
   > n3 = ToggleIconVisibility!(execute, self:Fort Replay Context Object Reference=Get FortReplayContext(), ToToggle:Object Class Reference=@n2) -> (then, ReturnValue:bool)
   > SetRenderOpacity(
    self=Get ButtonClicked() -> (ButtonClicked:Common Button Legacy Object Reference)
    InOpacity=Select(Option 0:double=0.4, Option 1:double=1, Index=@n3) -> (ReturnValue:double)
   )
  ]}
 Refresh Options {nodes [
   Refresh Options() {is=title K2Node_FunctionEntry}
   > n1 = Cast To FortPlayerController(Object=GetOwningPlayer())
   > n2 = Cast To FortControllerComponent_MinimapSpectating(Object=GetMiniMapComponent(self=@n1))
   > n3 = Set MinimapComponent(MinimapComponent:Fort Controller Component Minimap Spectating Object Reference=@n2) -> (Output_Get:Fort Controller Component Minimap Spectating Object Reference)
   > Update Row Selector!(
    execute
    self:Rotator Selector Replay Settings Object Reference=Get WeaponTracersRotator()
    Currently Selected:int=SelectInt(A=1, B=0, bPickA=GetMapWeaponTracersEnabled!(self:Fort Controller Component Minimap Spectating Object Reference=@n3) -> (ReturnValue:bool))
   ) -> (then)
   n4 = GetMaxMapZoomLevel(self=Get FortReplayContext())
   n5 = Get MinimapComponent() -> (MinimapComponent:Fort Controller Component Minimap Spectating Object Reference)
   n6 = Get NameZoomVisibility()
   > Update Slider(
    self=@n6
    Slider Text="GetText (Text)"!(self:Text Object Reference=Get Text_SettingValue!(self:Options Menu Slider Replay Object Reference=@n6) -> (Text_SettingValue:Common Text Object Reference)) -> (ReturnValue:Text)
    Slider Value=MapRangeClamped(
     Value=Get ZoomThresholdForPlayerNames!(self:Fort Controller Component Minimap Spectating Object Reference=@n5) -> (ZoomThresholdForPlayerNames:float)
     InRangeA=1
     InRangeB=float + float(A=@n4, B=0.01)
     OutRangeA=1
     OutRangeB=0
    )
    Min Max Value=(X=0,Y=1)
   )
   n7 = Get MinimapComponent()
   > Update Row Selector(
    self=Get VehicleIconsRotator()
    Currently Selected=SelectInt(A=1, B=0, bPickA=GetShowVehiclesOnMap!(self:Fort Controller Component Minimap Spectating Object Reference=@n7) -> (ReturnValue:bool))
   )
   n8 = GetShowPlayersOnMap!(self:Fort Controller Component Minimap Spectating Object Reference=Get MinimapComponent()) -> (ReturnValue:bool)
   > Update Row Selector(
    self=Cast To RotatorSelectorReplaySettings(Object=Get PlayerIndicatorsRotator()) -> (bSuccess=~)
    Currently Selected=SelectInt(A=1, B=0, bPickA=@n8)
   )
   > SetVisibility(
    self=[Get WeaponTracersRotator(), Get NameZoomVisibility(), Get VehicleIconsRotator()]
    InVisibility=Select(Option 0:ESlateVisibility Enum=Collapsed, Option 1:ESlateVisibility Enum=Visible, Index=@n8) -> (ReturnValue:ESlateVisibility Enum)
   )
   n9 = Get MinimapComponent()
   > Update Row Selector(
    self=Get BusPathsRotator()
    Currently Selected=SelectInt(A=1, B=0, bPickA=GetShowBusPathsOnMap!(self:Fort Controller Component Minimap Spectating Object Reference=@n9) -> (ReturnValue:bool))
   )
  ]}
]
variables [MapIconVisibility {default {}}]
widgets {
 info {inheritedWidgetCount=1, parentClass=/Script/FortniteUI.FortBroadcasterMapSettings, rootWidgetClass=&VerticalBox, widgetCount=10}
 widgets [
  VerticalBox_2 {
   widgetClassPath=&VerticalBox
   widget=VerticalBox_2
   properties {navigation=&self:WidgetTree.VerticalBox_2.WidgetNavigation_0}
  }
  Header {
   widgetClassPath=&CommonTextBlock
   parent=VerticalBox_2
   widget=Header
   properties {
    style=/FortUILibrary/Styles/Typography/Foundation/TextStyle-Header-M.TextStyle-Header-M_C
    text=Map Settings
    textTransformPolicy=ToUpper
    font {
     bForceMonospaced=false
     bMaterialIsStencil=false
     fontMaterial=null
     fontObject=&Burbank
     letterSpacing=0
     monospacedWidth=1
     size=32
     skewAmount=0
     typefaceFontName=Burbank-Black
     outlineSettings {
      bApplyOutlineToDropShadows=false
      bMiteredCorners=false
      bSeparateFillAlpha=false
      outlineMaterial=null
      outlineSize=0
      outlineColor=<rgba 0 0 0 1>
     }
    }
    shadowOffset=<0 0>
   }
   slot {class=VerticalBoxSlot, properties {horizontalAlignment=HAlign_Center}}
  }
  PlayerIndicatorsRotator {
   widgetClassPath=/FortniteCoreUI/UI/Replay/RotatorSelectorReplaySettings
   bInherited=true
   bIsVariable=true
   parent=VerticalBox_2
   widget=PlayerIndicatorsRotator
   slot {class=VerticalBoxSlot}
  }
  WeaponTracersRotator {
   widgetClassPath=&RotatorSelectorReplaySettings
   bIsVariable=true
   parent=VerticalBox_2
   widget=WeaponTracersRotator
   slot {class=VerticalBoxSlot}
  }
  VehicleIconsRotator {
   widgetClassPath=&RotatorSelectorReplaySettings
   bIsVariable=true
   parent=VerticalBox_2
   widget=VehicleIconsRotator
   slot {class=VerticalBoxSlot}
  }
  BusPathsRotator {
   widgetClassPath=&RotatorSelectorReplaySettings
   bIsVariable=true
   parent=VerticalBox_2
   widget=BusPathsRotator
   slot {class=VerticalBoxSlot}
  }
  NameZoomVisibility {
   widgetClassPath=/FortniteCoreUI/UI/Replay/OptionsMenuSliderReplay
   bIsVariable=true
   parent=VerticalBox_2
   widget=NameZoomVisibility
   slot {class=VerticalBoxSlot}
  }
  CommonTextBlock_0 {
   widgetClassPath=&CommonTextBlock
   parent=VerticalBox_2
   widget=CommonTextBlock_0
   properties {
    scrollStyle=&TextScrollStyle-Base_C
    style=/FortUILibrary/Styles/Typography/Foundation/TextStyle-Base-S.TextStyle-Base-S_C
    text=Map Icons
    font {
     bForceMonospaced=false
     bMaterialIsStencil=false
     fontMaterial=null
     fontObject=&NotoSans
     letterSpacing=0
     monospacedWidth=1
     size=18
     skewAmount=0
     typefaceFontName=NotoSans-Regular
     outlineSettings {
      bApplyOutlineToDropShadows=false
      bMiteredCorners=false
      bSeparateFillAlpha=false
      outlineMaterial=null
      outlineSize=0
      outlineColor=<rgba 0 0 0 1>
     }
    }
    shadowOffset=<0 0>
   }
   slot {class=VerticalBoxSlot}
  }
  SizeBox_0 {
   widgetClassPath=&SizeBox
   parent=VerticalBox_2
   widget=SizeBox_0
   properties {bOverride_HeightOverride=true, heightOverride=36}
   slot {class=VerticalBoxSlot}
  }
  ButtonBox {widgetClassPath=&HorizontalBox, bIsVariable=true, parent=SizeBox_0, widget=ButtonBox, slot {class=SizeBoxSlot}}
 ]
}
